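# Looping stereo MetaSound source for the Feline vehicle engine, fed by Fort.Vehicle and REV inputs.
# A REV player renders the revving engine, with pitch following bonus speed and grounded state.
# Its fundamental frequency and gear drive crossfade managers and pitch for looped idle, thrust base and high thrust layers, plus a throttle-scaled throttle-on layer.
# All layers mix to stereo, with character affiliation adjusting thrust-base gain.
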MetaSoundSource /VehicleAudio/EngineAudio/Feline/MSS_EngineAudio_Feline
inputs [
 Character.Affiliation {type=Enum:FortAudioAffiliation}
 Character.Health {type=Float}
 Character.UniqueID {type=Int32}
 Fort.Vehicle.Acceleration {type=Float}
 Fort.Vehicle.BlowoffStrength {type=Float}
 Fort.Vehicle.BonusSpeed {type=Float}
 Fort.Vehicle.EnableShifting {type=Bool}
 Fort.Vehicle.Gear {type=Int32}
 Fort.Vehicle.IsGrounded {type=Bool}
 Fort.Vehicle.Load {type=Float}
 Fort.Vehicle.LowFuel {type=Bool}
 Fort.Vehicle.NormalSpeed {type=Float}
 Fort.Vehicle.OnBlowoff {type=Trigger}
 Fort.Vehicle.OnDownshift {type=Trigger}
 Fort.Vehicle.OnThrottleEngaged {type=Trigger}
 Fort.Vehicle.OnThrottleReleased {type=Trigger}
 Fort.Vehicle.OnUpshift {type=Trigger}
 Fort.Vehicle.Pitch {type=Float}
 Fort.Vehicle.RPM {type=Float}
 Fort.Vehicle.Throttle {type=Float}
 Rev Asset {type=Rev Asset, access=Constructor}
 Rev Settings {type=Rev Settings, access=Constructor}
 REV.EnableShifting {type=Bool}
 REV.Gear {type=Int32}
 REV.RPM {type=Float}
 REV.Throttle {type=Float}
 REV.Velocity {type=Float}
 UE.Attenuation.Distance {type=Float}
 UE.Source.OnPlay {type=Trigger}
]
interfaces [Fort.Character:1.0, Fort.Vehicle:1.0, REV:1.0, UE.Attenuation:1.0, UE.OutputFormat.Stereo:1.0, UE.Source:1.1]
outputs [
 "UE.OutputFormat.Stereo.Audio:0" {type=Audio}
 "UE.OutputFormat.Stereo.Audio:1" {type=Audio}
]
pages [Default {
  input_defaults {
   Character.Affiliation=0
   Character.Health=0
   Character.UniqueID=0
   Fort.Vehicle.Acceleration=0
   Fort.Vehicle.BlowoffStrength=0
   Fort.Vehicle.BonusSpeed=0.004
   Fort.Vehicle.EnableShifting=null
   Fort.Vehicle.Gear=0
   Fort.Vehicle.IsGrounded=null
   Fort.Vehicle.Load=0
   Fort.Vehicle.LowFuel=null
   Fort.Vehicle.NormalSpeed=0
   Fort.Vehicle.OnBlowoff=null
   Fort.Vehicle.OnDownshift=null
   Fort.Vehicle.OnThrottleEngaged=null
   Fort.Vehicle.OnThrottleReleased=null
   Fort.Vehicle.OnUpshift=null
   Fort.Vehicle.Pitch=0
   Fort.Vehicle.RPM=0
   Fort.Vehicle.Throttle=0
   REV.EnableShifting=null
   REV.Gear=1
   REV.RPM=0
   REV.Throttle=0
   REV.Velocity=0
   Rev Asset=&/EngineAudio_Feline_Model
   Rev Settings=&/REVSettings_Feline
   UE.Attenuation.Distance=0
   UE.Source.OnPlay=null
  }
  nodes [
   n1 = VariableAccessor.Float(Variable=VariableMutator.Float(Variable=InitVariable.Float().Variable, Value=Max.Max:Float(B=^Fort.Vehicle.Throttle, A=^REV.Throttle).Value).Variable)
   n2 = VariableAccessor.Float(Variable=@n1.Variable)
   n3 = MapRange.MapRange:Float(In=^Fort.Vehicle.NormalSpeed) {also {defaults {Out Range B=2}}}
   n4 = VariableAccessor.Int32(Variable=VariableMutator.Int32(Variable=InitVariable.Int32().Variable, Value=Max.Max:Int32(B=^Fort.Vehicle.Gear, A=^REV.Gear).Value).Variable)
   n5 = VariableAccessor.Int32(Variable=@n4.Variable)
   n6 = Convert.Float:Int32(Int32=@n5.Value)
   n7 = MapRange.MapRange:Float(In=^Fort.Vehicle.BonusSpeed) {also {defaults {Out Range B=2}}}
   TriggerOnValueChange_0 = UE.Trigger On Value Change:Bool(Value=^Fort.Vehicle.IsGrounded)
   n8 = TriggerCompare.Clamp:Bool(A=^Fort.Vehicle.IsGrounded, Compare=@TriggerOnValueChange_0.Trigger) {also {defaults {B="<MetaSoundBuilderResult.SUCCEEDED: 0>"}}}
   n9 = VariableAccessor.Trigger(Variable=VariableMutator.Trigger(Variable=InitVariable.Trigger().Variable, Value=@n8.True).Variable)
   n10 = VariableAccessor.Trigger(Variable=VariableMutator.Trigger(Variable=InitVariable.Trigger().Variable, Value=@n8.False).Variable)
   TriggerRouteTriggerRouteFloat2_0 = "TriggerRoute.Trigger Route (Float, 2):Float"(Set 1=@n9.Value, Set 0=@n10.Value) {also {defaults {Value 0=5}}}
   n11 = VariableAccessor.Trigger(Variable=@n10.Variable)
   n12 = VariableAccessor.Trigger(Variable=@n9.Variable)
   TriggerRouteTriggerRouteTime2_1 = "TriggerRoute.Trigger Route (Time, 2):Time"(Set 0=@n11.Value, Set 1=@n12.Value) {also {defaults {Value 0=0.4, Value 1=0.1}}}
   n13 = UE.InterpTo:Audio(
    Target=UE.Add:Float(AdditionalOperands=UE.Reroute(Value=@n7.Out Value).Value, PrimaryOperand=@TriggerRouteTriggerRouteFloat2_0.Value).Out
    Interp Time=@TriggerRouteTriggerRouteTime2_1.Value
   )
   n14 = UE.Add:Float(
    AdditionalOperands=UE.InterpTo:Audio(Target=MapRange.MapRange:Float(In=@n6.Float) {also {defaults {In Range A=1, In Range B=4, Out Range B=5}}}.Out Value) {also {defaults {Interp Time=2}}}.Value
    PrimaryOperand=UE.Reroute(Value=@n13.Value).Value
   )
   n15 = VariableAccessor.Float(Variable=VariableMutator.Float(Variable=InitVariable.Float().Variable, Value=@n13.Value).Variable)
   TriggerRouteTriggerRouteTime2_0 = "TriggerRoute.Trigger Route (Time, 2):Time"(Set 0=VariableAccessor.Trigger(Variable=@n11.Variable).Value, Set 1=VariableAccessor.Trigger(Variable=@n12.Variable).Value) {also {defaults {Value 0=0.3, Value 1=0.1}}}
   n16 = Convert.Bool:Int32(Int32=Max.Max:Int32(B=Convert.Int32:Bool(Bool=^Fort.Vehicle.EnableShifting).Int32, A=Convert.Int32:Bool(Bool=^REV.EnableShifting).Int32).Value)
   n17 = InitVariable.Bool()
   n18 = InitVariable.Float()
   n19 = Max.Max:Float(B=^Fort.Vehicle.RPM, A=^REV.RPM)
   REVRevPlayer_0 = REV.Rev Player:Stereo(
    Rev Asset=^Rev Asset
    Rev Settings=^Rev Settings
    Play=^UE.Source.OnPlay
    Pitch Shift=UE.InterpTo:Audio(Target=UE.Reroute(Value=@n15.Value).Value, Interp Time=@TriggerRouteTriggerRouteTime2_0.Value).Value
    Enable Shifting=VariableAccessor.Bool(Variable=VariableMutator.Bool(Value=@n16.Bool, Variable=@n17.Variable).Variable).Value
    RPM=VariableAccessor.Float(Variable=VariableMutator.Float(Variable=@n18.Variable, Value=@n19.Value).Variable).Value
    Throttle=@n1.Value
    Gear=@n4.Value
   )
   n20 = VariableAccessor.Float(Variable=VariableMutator.Float(Variable=InitVariable.Float().Variable, Value=@REVRevPlayer_0.Fundamental Frequency).Variable)
   n21 = VariableAccessor.Float(Variable=VariableMutator.Float(
     Variable=InitVariable.Float().Variable
     Value=MapRange.MapRange:Float(In=@n20.Value) {also {defaults {In Range A=7.42, In Range B=34.97}}}.Out Value
    ).Variable)
   n22 = VariableAccessor.Float(Variable=@n21.Variable)
   n23 = UE.Add:Float(
    PrimaryOperand=@n14.Out
    AdditionalOperands=MapRange.MapRange:Float(In=@n22.Value) {also {defaults {In Range A=0.4, Out Range A=-1}}}.Out Value
   )
   n24 = VariableAccessor.Float(Variable=InitVariable.Float().Variable)
   n25 = VariableAccessor.Float(Variable=VariableMutator.Float(Value=UE.Add:Float(PrimaryOperand=@n23.Out, AdditionalOperands=@n24.Value).Out, Variable=InitVariable.Float().Variable).Variable)
   n26 = VariableAccessor.Int32(Variable=@n5.Variable)
   n27 = VariableAccessor.Int32(Variable=@n26.Variable)
   n28 = Convert.Float:Int32(Int32=@n27.Value)
   n29 = MapRange.MapRange:Float(In=@n28.Float) {also {defaults {In Range A=1, In Range B=4}}}
   n30 = VariableAccessor.Float(Variable=@n22.Variable)
   n31 = MapRange.MapRange:Float(In=UE.Add:Float(
     PrimaryOperand=UE.Multiply:Float(PrimaryOperand=@n29.Out Value) {also {defaults {AdditionalOperands=2}}}.Out
     AdditionalOperands=UE.Multiply:Float(PrimaryOperand=@n30.Value).Out
    ).Out) {also {defaults {In Range B=3}}}
   n32 = VariableAccessor.Float(Variable=VariableMutator.Float(Variable=InitVariable.Float().Variable, Value=UE.InterpTo:Audio(Target=@n31.Out Value) {also {defaults {Interp Time=1}}}.Value).Variable)
   n33 = Fortnite.CrossfadeSegmentManager:Float(In=@n32.Value) {also {defaults {Max=0.7, MaxRange=0.4, Min=-0.1, MinRange=0}}}
   WavePlayer_1 = UE.Wave Player:Mono(
    Pitch Shift=UE.Add:Float(PrimaryOperand=UE.Add:Float(AdditionalOperands=@n3.Out Value, PrimaryOperand=@n25.Value).Out) {also {defaults {AdditionalOperands=2}}}.Out
    Play=@n33.Activate
    Stop=@n33.Deactivate
   ) {also {
     defaults {Loop="<MetaSoundBuilderResult.SUCCEEDED: 0>", Wave Asset=&/EngineAudio_Feline_Idle_Prime}
     version [1,1]
    }}
   BiquadFilter_0 = UE.Biquad Filter:Audio(Gain=MapRange.MapRange:Float(In=@n2.Value) {also {defaults {Out Range B=8}}}.Out Value, In=@WavePlayer_1.Out Mono) {also {defaults {Cutoff Frequency=5000, Type=4}}}
   n34 = InitVariable.Bool()
   TriggerOnValueChange_2 = UE.Trigger On Value Change:Int32(Value=Convert.Int32:Enum:FortAudioAffiliation("Enum:FortAudioAffiliation"=^Character.Affiliation).Int32)
   TriggerAnyTriggerAny2_0 = "TriggerAny.Trigger Any (2)"(In 1=^UE.Source.OnPlay, In 0=@TriggerOnValueChange_2.Trigger)
   n35 = &TriggerCompareAffiliation(A=UE.Reroute(Value=^Character.Affiliation).Value, Compare=@TriggerAnyTriggerAny2_0.Out)
   TriggerToggle_0 = UE.Trigger Toggle(Off=@n35.False, On=@n35.True)
   n36 = Convert.Float:Bool(Bool=VariableAccessor.Bool(Variable=VariableMutator.Bool(Variable=@n34.Variable, Value=@TriggerToggle_0.Value).Variable).Value)
   DecibelstoLinearGain_0 = UE.Decibels to Linear Gain:Float(Decibels=MapRange.MapRange:Float(In=UE.InterpTo:Audio(Target=@n36.Float) {also {defaults {Interp Time=0.5}}}.Value) {also {defaults {In Range A=1, In Range B=0, Out Range B=4}}}.Out Value)
   n37 = VariableAccessor.Float(Variable=@n32.Variable)
   n38 = VariableAccessor.Float(Variable=@n37.Variable)
   n39 = VariableAccessor.Float(Variable=@n38.Variable)
   n40 = Fortnite.CrossfadeSegmentManager:Float(In=VariableAccessor.Float(Variable=@n39.Variable).Value) {also {defaults {Max=1.1, MaxRange=0, Min=0.2, MinCurve=0.6, MinRange=0.5}}}
   n41 = UE.Add:Float(
    PrimaryOperand=UE.Add:Float(AdditionalOperands=^Fort.Vehicle.BonusSpeed, PrimaryOperand=^Fort.Vehicle.NormalSpeed).Out
    AdditionalOperands=UE.Multiply:Float(PrimaryOperand=@n21.Value) {also {defaults {AdditionalOperands=2}}}.Out
   )
   n42 = InitVariable.Float()
   TriggerOnValueChange_1 = UE.Trigger On Value Change:Int32(Value=@n26.Value)
   TriggerAnyTriggerAny2_1 = "TriggerAny.Trigger Any (2)"(In 1=^UE.Source.OnPlay, In 0=@TriggerOnValueChange_1.Trigger)
   n43 = TriggerCompare.Clamp:Int32(Compare=@TriggerAnyTriggerAny2_1.Out, A=VariableAccessor.Int32(Variable=@n27.Variable).Value) {also {defaults {B=4, Type=5}}}
   TriggerRouteTriggerRouteFloat2_1 = "TriggerRoute.Trigger Route (Float, 2):Float"(
    Value 0=MapRange.MapRange:Float(In=Convert.Float:Int32(Int32=@n26.Value).Float) {also {defaults {In Range A=4, In Range B=6, Out Range A=0.8, Out Range B=0.4}}}.Out Value
    Set 1=@n43.False
    Set 0=@n43.True
   ) {also {defaults {Value 1=1}}}
   n44 = UE.InterpTo:Audio(Target=@TriggerRouteTriggerRouteFloat2_1.Value) {also {defaults {Interp Time=3}}}
   n45 = UE.Multiply:Float(
    PrimaryOperand=UE.Multiply:Float(
     AdditionalOperands=MapRange.MapRange:Float(In=@n38.Value) {also {defaults {Out Range A=0.5}}}.Out Value
     PrimaryOperand=MapRange.MapRange:Float(In=@n41.Out) {also {defaults {In Range B=4, Out Range A=0.5}}}.Out Value
    ).Out
    AdditionalOperands=VariableAccessor.Float(Variable=VariableMutator.Float(Variable=@n42.Variable, Value=@n44.Value).Variable).Value
   )
   n46 = VariableAccessor.Float(Variable=VariableMutator.Float(Variable=InitVariable.Float().Variable, Value=UE.InterpTo:Audio(Target=@n45.Out) {also {defaults {Interp Time=1}}}.Value).Variable)
   n47 = VariableAccessor.Float(Variable=@n46.Variable)
   ADSREnvelopeADSREnvelope_0 = ADSR Envelope.ADSR Envelope:Float(Trigger Attack=^UE.Source.OnPlay) {also {defaults {Attack Time=2, Sustain Level=1}}}
   n48 = MapRange.MapRange:Float(In=@n28.Float) {also {defaults {In Range A=3, In Range B=5}}}
   n49 = MapRange.MapRange:Float(In=UE.Add:Float(AdditionalOperands=UE.Multiply:Float(PrimaryOperand=@n48.Out Value).Out, PrimaryOperand=UE.Multiply:Float(PrimaryOperand=@n30.Value).Out).Out) {also {defaults {In Range B=2}}}
   n50 = VariableAccessor.Float(Variable=VariableMutator.Float(Variable=InitVariable.Float().Variable, Value=UE.InterpTo:Audio(Target=@n49.Out Value) {also {defaults {Interp Time=1}}}.Value).Variable)
   n51 = UE.Divide:Float(PrimaryOperand=UE.Add:Float(PrimaryOperand=@n37.Value, AdditionalOperands=VariableAccessor.Float(Variable=@n50.Variable).Value).Out) {also {defaults {AdditionalOperands=2}}}
   n52 = VariableAccessor.Float(Variable=@n25.Variable)
   WavePlayer_2 = UE.Wave Player:Mono(
    Pitch Shift=UE.Add:Float(AdditionalOperands=MapRange.MapRange:Float(In=@n51.Out) {also {defaults {Out Range B=11}}}.Out Value, PrimaryOperand=@n52.Value).Out
    Play=@n40.Activate
    Stop=@n40.Deactivate
   ) {also {
     defaults {Loop="<MetaSoundBuilderResult.SUCCEEDED: 0>", Wave Asset=&/EngineAudio_Feline_ThrustBase_Prime}
     version [1,1]
    }}
   n53 = UE.Reroute(Value=@WavePlayer_2.Out Mono)
   n54 = Fortnite.CrossfadeSegmentManager:Float(In=@n50.Value) {also {defaults {Max=1.1, MaxRange=0, MinCurve=0.4, MinRange=0.3}}}
   n55 = VariableAccessor.Float(Variable=@n47.Variable)
   n56 = VariableAccessor.Float(Variable=@n52.Variable)
   n57 = Fortnite.CrossfadeSegmentManager:Float(In=@n39.Value) {also {defaults {Max=0.6, MaxRange=0.4, Min=-0.1, MinRange=0}}}
   WavePlayer_0 = UE.Wave Player:Mono(
    Pitch Shift=UE.Add:Float(
     AdditionalOperands=MapRange.MapRange:Float(In=^Fort.Vehicle.NormalSpeed) {also {defaults {In Range B=0.4, Out Range B=3}}}.Out Value
     PrimaryOperand=VariableAccessor.Float(Variable=@n56.Variable).Value
    ).Out
    Play=@n57.Activate
    Stop=@n57.Deactivate
   ) {also {
     defaults {Loop="<MetaSoundBuilderResult.SUCCEEDED: 0>", Wave Asset=&/EngineAudio_Feline_ThrottleOn_Prime}
     version [1,1]
    }}
   n58 = UE.Reroute(Value=@WavePlayer_0.Out Mono)
   n59 = UE.Multiply:Float(
    AdditionalOperands=Clamp.Clamp:Float(In=UE.Multiply:Float(PrimaryOperand=^Fort.Vehicle.NormalSpeed) {also {defaults {AdditionalOperands=5}}}.Out) {also {defaults {Max=1}}}.Value
    PrimaryOperand=UE.InterpTo:Audio(Target=VariableAccessor.Float(Variable=@n2.Variable).Value).Value
   )
   n60 = UE.Multiply:Float(
    AdditionalOperands=@n57.Volume
    PrimaryOperand=UE.InterpTo:Audio(Target=Clamp.Clamp:Float(In=@n59.Out) {also {defaults {Max=1}}}.Value).Value
   )
   n61 = UE.Reroute(Value=@REVRevPlayer_0.Out Left)
   WavePlayer_3 = UE.Wave Player:Stereo(
    Pitch Shift=UE.Add:Float(PrimaryOperand=MapRange.MapRange:Float(In=@n50.Value) {also {defaults {Out Range B=8}}}.Out Value, AdditionalOperands=@n56.Value).Out
    Play=@n54.Activate
    Stop=@n54.Deactivate
   ) {also {
     defaults {Loop="<MetaSoundBuilderResult.SUCCEEDED: 0>", Wave Asset=&/EngineAudio_Feline_ThrustHiWide_Prime}
     version [1,1]
    }}
   AudioMixerAudioMixerStereo6_0 = "AudioMixer.Audio Mixer (Stereo, 6)"(
    In 1 L=@BiquadFilter_0.Out
    In 1 R=@BiquadFilter_0.Out
    Gain 2=UE.Multiply:Float(
     AdditionalOperands=@DecibelstoLinearGain_0.Linear Gain
     PrimaryOperand=UE.Multiply:Float(AdditionalOperands=@n40.Volume, PrimaryOperand=@n47.Value).Out
    ).Out
    Gain 1=UE.Multiply:Float(
     PrimaryOperand=@ADSREnvelopeADSREnvelope_0.Out Envelope
     AdditionalOperands=UE.Multiply:Float(AdditionalOperands=@n33.Volume, PrimaryOperand=@n46.Value).Out
    ).Out
    In 2 L=@n53.Value
    In 2 R=@n53.Value
    Gain 3=UE.Reroute(Value=UE.Multiply:Float(AdditionalOperands=@n54.Volume, PrimaryOperand=@n55.Value).Out).Value
    In 5 L=@n58.Value
    In 5 R=@n58.Value
    Gain 5=UE.Reroute(Value=UE.Multiply:Float(PrimaryOperand=@n60.Out) {also {defaults {AdditionalOperands=0.7}}}.Out).Value
    In 0 L=@n61.Value
    In 0 R=@n61.Value
    In 3 L=@WavePlayer_3.Out Left
    In 3 R=@WavePlayer_3.Out Right
   )
   InitVariable.Trigger()
   InitVariable.Trigger()
   InitVariable.Trigger()
   VariableMutator.Float(Value=@n14.Out, Variable=InitVariable.Float().Variable)
   VariableMutator.Trigger(
    Variable=InitVariable.Trigger().Variable
    Value=UE.TriggerOnThreshold:Float(In=^Fort.Vehicle.BonusSpeed) {also {defaults {Threshold=0.1}}}.Out
   )
   VariableMutator.Trigger(
    Variable=InitVariable.Trigger().Variable
    Value=UE.TriggerOnThreshold:Float(In=^Fort.Vehicle.BonusSpeed) {also {defaults {Threshold=0.1, Type=1}}}.Out
   )
  ]
  links [
   {dst="@UE.OutputFormat.Stereo.Audio:0", src=AudioMixerAudioMixerStereo6_0.Out L}
   {dst="@UE.OutputFormat.Stereo.Audio:1", src=AudioMixerAudioMixerStereo6_0.Out R}
  ]
 }]
properties {
 assetImportData=&self:AssetImportData
 attenuationSettings=&Vehicle_Engine_Att
 audioPropertiesSheet=&APS_Cosmetic_Vehicle_Engine
 bLooping=true
 duration=10000
 maxDistance=12000
 numChannels=2
 outputFormat=Stereo
 priority=80
 sampleRate=48000
 soundClassObject=&Fort_Vehicles_Engine
 soundSubmixObject=&Fortnite_SFX
 volume=1.35
 concurrencySet [&Conc_Veh_EngineAudio_LowSpec, &Conc_Veh_EngineAudio_LowSpec_Actor]
 modulationSettings {
  highpassRouting=Union
  lowpassRouting=Union
  pitchRouting=Union
  volumeRouting=Union
  highpassModulationDestination {bEnableModulation=true, modulators [&PP_FortVehicles_Engine_HPF]}
  lowpassModulationDestination {bEnableModulation=true, modulators [&Patch_FortVehicles_Engine_LPF]}
  pitchModulationDestination {bEnableModulation=true, modulators [&DM_Bus_Vehicle_Engine_Pitch]}
  volumeModulationDestination {
   bEnableModulation=true
   modulators [&DM_VehicleEnginePatch, &Patch_FortVehicles_Engine_Vol]
  }
 }
}
