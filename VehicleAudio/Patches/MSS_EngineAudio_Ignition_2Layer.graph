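# One-shot mono MetaSound source for vehicle engine ignition, blending a main sound with a distance-shaped additional layer.
# On play, each wave array passes a size check, then a random pick feeds its own wave player.
# Listener distance drives the added layer's volume offset and log-scaled low-pass and high-pass cutoffs before both layers are mixed.
# OnFinished fires through a trigger accumulator once both players finish.
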MetaSoundSource /VehicleAudio/Patches/MSS_EngineAudio_Ignition_2Layer
inputs [
 IgnitionAddLayer {type=WaveAsset:Array, access=Constructor}
 IgnitionAddLayerHPFCutoffMax {type=Float}
 IgnitionAddLayerHPFCutoffMin {type=Float}
 IgnitionAddLayerLPFCutoffMax {type=Float}
 IgnitionAddLayerLPFCutoffMin {type=Float}
 IgnitionAddLayerMaxDistance {type=Float}
 IgnitionAddLayerMinDistance {type=Float}
 IgnitionAddLayerVolume {type=Float}
 IgnitionAddLayerVolumeOffsetMax {type=Float}
 IgnitionAddLayerVolumeOffsetMin {type=Float}
 IgnitionSound {type=WaveAsset:Array, access=Constructor}
 IgnitionSoundVolume {type=Float, access=Constructor}
 UE.Attenuation.Distance {type=Float}
 UE.Source.OnPlay {type=Trigger}
]
interfaces [UE.Attenuation:1.0, UE.OutputFormat.Mono:1.0, UE.Source.OneShot:1.0, UE.Source:1.1]
outputs [
 "UE.OutputFormat.Mono.Audio:0" {type=Audio}
 UE.Source.OneShot.OnFinished {type=Trigger}
]
pages [Default {
  input_defaults {
   IgnitionAddLayerHPFCutoffMax=500
   IgnitionAddLayerHPFCutoffMin=20
   IgnitionAddLayerLPFCutoffMax=3000
   IgnitionAddLayerLPFCutoffMin=20000
   IgnitionAddLayerMaxDistance=6000
   IgnitionAddLayerMinDistance=0
   IgnitionAddLayerVolume=0
   IgnitionAddLayerVolumeOffsetMax=-10
   IgnitionAddLayerVolumeOffsetMin=0
   IgnitionSoundVolume=0
   UE.Attenuation.Distance=0
   UE.Source.OnPlay=null
   IgnitionAddLayer []
   IgnitionSound []
  }
  nodes [
   n1 = VariableAccessor.Float(Variable=VariableMutator.Float(
     Variable=InitVariable.Float().Variable
     Value=MapRange.MapRange:Float(In Range B=^IgnitionAddLayerMaxDistance, In Range A=^IgnitionAddLayerMinDistance, In=^UE.Attenuation.Distance).Out Value
    ).Variable)
   n2 = VariableAccessor.Float(Variable=@n1.Variable)
   n3 = VariableAccessor.Float(Variable=@n2.Variable)
   DecibelstoLinearGain_0 = UE.Decibels to Linear Gain:Float(Decibels=UE.Add:Float(
     PrimaryOperand=^IgnitionAddLayerVolume
     AdditionalOperands=MapRange.MapRange:Float(Out Range B=^IgnitionAddLayerVolumeOffsetMax, Out Range A=^IgnitionAddLayerVolumeOffsetMin, In=@n3.Value).Out Value
    ).Out)
   DecibelstoLinearGain_1 = UE.Decibels to Linear Gain:Float(Decibels=^IgnitionSoundVolume)
   ArrayRandomGet_0 = Array.Random Get:WaveAsset:Array(
    In Array=^IgnitionSound
    Next=TriggerCompare.Clamp:Int32(Compare=^UE.Source.OnPlay, A=Array.Num:WaveAsset:Array(Array=^IgnitionSound).Num) {also {defaults {B=0, Type=1}}}.TRUE
   )
   WavePlayer_1 = UE.Wave Player:Mono(Play=@ArrayRandomGet_0.On Next, Wave Asset=@ArrayRandomGet_0.Value) {also {version [1,1]}}
   LinearToLogFrequency_1 = UE.Linear To Log Frequency(Max Range=^IgnitionAddLayerLPFCutoffMax, Min Range=^IgnitionAddLayerLPFCutoffMin, Value=@n1.Value)
   ArrayRandomGet_1 = Array.Random Get:WaveAsset:Array(
    In Array=^IgnitionAddLayer
    Next=TriggerCompare.Clamp:Int32(Compare=^UE.Source.OnPlay, A=Array.Num:WaveAsset:Array(Array=^IgnitionAddLayer).Num) {also {defaults {B=0, Type=1}}}.TRUE
   ) {also {defaults {Weights []}}}
   WavePlayer_0 = UE.Wave Player:Mono(Play=@ArrayRandomGet_1.On Next, Wave Asset=@ArrayRandomGet_1.Value) {also {version [1,1]}}
   BiquadFilter_1 = UE.Biquad Filter:Audio(Cutoff Frequency=@LinearToLogFrequency_1.Frequency, In=@WavePlayer_0.Out Mono)
   LinearToLogFrequency_0 = UE.Linear To Log Frequency(Max Range=^IgnitionAddLayerHPFCutoffMax, Min Range=^IgnitionAddLayerHPFCutoffMin, Value=@n2.Value)
   BiquadFilter_0 = UE.Biquad Filter:Audio(In=@BiquadFilter_1.Out, Cutoff Frequency=@LinearToLogFrequency_0.Frequency) {also {defaults {Type=1}}}
   AudioMixerAudioMixerMono3_0 = "AudioMixer.Audio Mixer (Mono, 3)"(
    Gain 1=@DecibelstoLinearGain_0.Linear Gain
    Gain 0=@DecibelstoLinearGain_1.Linear Gain
    In 0=UE.Reroute(Value=@WavePlayer_1.Out Mono).Value
    In 1=UE.Reroute(Value=@BiquadFilter_0.Out).Value
   )
   TriggerAccumulatorTriggerAccumulate2_0 = "TriggerAccumulator.Trigger Accumulate (2)"(In 1=@WavePlayer_0.On Finished, In 0=@WavePlayer_1.On Finished)
  ]
  links [
   {dst="@UE.OutputFormat.Mono.Audio:0", src=AudioMixerAudioMixerMono3_0.Out}
   {dst="@UE.Source.OneShot.OnFinished", src=TriggerAccumulatorTriggerAccumulate2_0.Out}
  ]
 }]
properties {
 assetImportData=&self:AssetImportData
 attenuationSettings=/VehicleAudio/Settings/Vehicle_Movement_Att
 audioPropertiesSheet=/VehicleAudio/PropertySheets/APS_Cosmetic_Vehicle_Engine_Ignition
 maxDistance=12512
 sampleRate=48000
 soundClassObject=&Fort_Vehicles_Interior
 soundSubmixObject=&Fortnite_SFX
 concurrencySet [/VehicleAudio/Settings/Vehicle_Cosmetic_Ignition_Conc]
 modulationSettings {
  highpassRouting=Union
  lowpassRouting=Union
  volumeRouting=Union
  highpassModulationDestination {bEnableModulation=true, modulators [&PP_FortVehicles_Engine_HPF]}
  lowpassModulationDestination {bEnableModulation=true, modulators [&Patch_FortVehicles_Engine_LPF]}
  volumeModulationDestination {
   bEnableModulation=true
   modulators [&Patch_FortVehicles_Engine_Vol, &DM_VehicleEnginePatch]
  }
 }
}
