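# Mono one-shot MetaSound preset for the Feline vehicle engine ignition sound, built on the shared two-layer ignition parent.
# It supplies the main ignition and "Prime" add-layer waves, leaving playback logic to the parent.
# Add-layer inputs define distance ranges for volume offset and high/low-pass cutoffs, indicating a quieter, band-narrowed layer farther away.
# Vehicle attenuation, cosmetic ignition concurrency, and engine volume and filter modulators govern the source.
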MetaSoundSource /VehicleAudio/EngineAudio/Feline/MSS_EngineAudio_Feline_Ignition_2Layer_Preset
preset_parent=/VehicleAudio/Patches/MSS_EngineAudio_Ignition_2Layer
inputs [
 IgnitionAddLayer {type=WaveAsset:Array, access=Constructor}
 IgnitionAddLayerHPFCutoffMax {type=Float}
 IgnitionAddLayerHPFCutoffMin {type=Float}
 IgnitionAddLayerLPFCutoffMax {type=Float}
 IgnitionAddLayerLPFCutoffMin {type=Float}
 IgnitionAddLayerMaxDistance {type=Float}
 IgnitionAddLayerMinDistance {type=Float}
 IgnitionAddLayerVolume {type=Float}
 IgnitionAddLayerVolumeOffsetMax {type=Float}
 IgnitionAddLayerVolumeOffsetMin {type=Float}
 IgnitionSound {type=WaveAsset:Array, access=Constructor}
 IgnitionSoundVolume {type=Float, access=Constructor}
 UE.Attenuation.Distance {type=Float}
 UE.Source.OnPlay {type=Trigger}
]
interfaces [UE.Attenuation:1.0, UE.OutputFormat.Mono:1.0, UE.Source.OneShot:1.0, UE.Source:1.1]
outputs [
 "UE.OutputFormat.Mono.Audio:0" {type=Audio}
 UE.Source.OneShot.OnFinished {type=Trigger}
]
pages [Default {input_defaults {
   IgnitionAddLayerHPFCutoffMax=500
   IgnitionAddLayerHPFCutoffMin=20
   IgnitionAddLayerLPFCutoffMax=3000
   IgnitionAddLayerLPFCutoffMin=20000
   IgnitionAddLayerMaxDistance=6000
   IgnitionAddLayerMinDistance=0
   IgnitionAddLayerVolume=0
   IgnitionAddLayerVolumeOffsetMax=-10
   IgnitionAddLayerVolumeOffsetMin=0
   IgnitionSoundVolume=0
   UE.Attenuation.Distance=0
   UE.Source.OnPlay=null
   IgnitionAddLayer [&/EngineAudio_Feline_Ignition_Prime]
   IgnitionSound [&/EngineAudio_Feline_Ignition]
  }}]
properties {
 assetImportData=&self:AssetImportData
 attenuationSettings=/VehicleAudio/Settings/Vehicle_Movement_Att
 audioPropertiesSheet=/VehicleAudio/PropertySheets/APS_Cosmetic_Vehicle_Engine_Ignition
 maxDistance=12512
 sampleRate=48000
 soundClassObject=&Fort_Vehicles_Interior
 soundSubmixObject=&Fortnite_SFX
 concurrencySet [/VehicleAudio/Settings/Vehicle_Cosmetic_Ignition_Conc]
 modulationSettings {
  highpassRouting=Union
  lowpassRouting=Union
  volumeRouting=Union
  highpassModulationDestination {bEnableModulation=true, modulators [&PP_FortVehicles_Engine_HPF]}
  lowpassModulationDestination {bEnableModulation=true, modulators [&Patch_FortVehicles_Engine_LPF]}
  volumeModulationDestination {
   bEnableModulation=true
   modulators [&Patch_FortVehicles_Engine_Vol, &DM_VehicleEnginePatch]
  }
 }
}
